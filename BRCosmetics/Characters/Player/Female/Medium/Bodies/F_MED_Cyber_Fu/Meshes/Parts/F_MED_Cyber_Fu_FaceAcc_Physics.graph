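# Physics asset for the Cyber Fu female medium character's face accessory mesh, rigging left and right hair chains for physical motion.
# Each chain runs from the head through a tightly limited base bone into three capsule segments, with moderate swing, slightly wider at the tip, and minimal twist.
# Capsules on the head, clavicles and upper arms give body volumes around the hair, with free angular constraints joining each upper arm to its clavicle.
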PhysicsAsset /BRCosmetics/Characters/Player/Female/Medium/Bodies/F_MED_Cyber_Fu/Meshes/Parts/F_MED_Cyber_Fu_FaceAcc_Physics
bodies {
 base_hair_01_l [{shapeType=Capsule, length=1, radius=1, shapeName=base_hair_01_l_capsule, rotation=<rot 90 0 0>}]
 base_hair_01_r [{shapeType=Capsule, length=1, radius=1, shapeName=base_hair_01_r_capsule, rotation=<rot 90 0 0>}]
 clavicle_l [{shapeType=Capsule, length=5, radius=6, shapeName=clavicle_l_capsule, center=<4 0 0>, rotation=<rot 90 0 0>}]
 clavicle_r [{shapeType=Capsule, length=5, radius=6, shapeName=clavicle_l_capsule, center=<-4 0 0>, rotation=<rot 90 0 0>}]
 dyn_hair_01_l [{shapeType=Capsule, length=3.5, radius=1.5, shapeName=dyn_hair_01_l_capsule, center=<-3.8825 0 0>, rotation=<rot 75 0 5.0008>}]
 dyn_hair_01_r [{shapeType=Capsule, length=3.5, radius=1.5, shapeName=dyn_hair_01_r_capsule, center=<-3.6186 0 -0.1353>, rotation=<rot 80 0 0>}]
 dyn_hair_02_l [{shapeType=Capsule, length=3.5, radius=1.5, shapeName=dyn_hair_02_l_capsule, center=<-3.3065 0 0.1833>, rotation=<rot 65 -0.0001 0.0001>}]
 dyn_hair_02_r [{
   shapeType=Capsule
   length=3.5
   radius=1.5
   shapeName=dyn_hair_02_r_capsule
   center=<-2.9182 -0.7959 -0.163>
   rotation=<rot 75.0009 30.0018 9.9995>
  }]
 dyn_hair_03_l [{shapeType=Capsule, length=3, radius=1.5, shapeName=dyn_hair_03_l_capsule, center=<-2.9579 0 0.1011>, rotation=<rot 69.9996 -10.0016 -5.0001>}]
 dyn_hair_03_r [{
   shapeType=Capsule
   length=1.5
   radius=1.5
   shapeName=dyn_hair_03_r_capsule
   center=<-0.8433 -1.6515 -0.6741>
   rotation=<rot 69.9997 -120.0008 -5.0015>
  }]
 head [
  {shapeType=Capsule, length=5, radius=5.8, center=<2.0589 3.9694 0>, rotation=<rot 90 0 -15.0002>}
  {shapeType=Capsule, length=4.0938, radius=7, center=<0 -3.2162 0>, rotation=<rot 90 0 0>}
  {shapeType=Capsule, length=10.75, radius=6.5, shapeName=head_capsule, center=<7.0908 -0.3543 0>, rotation=<rot 90 0 60.0001>}
 ]
 upperarm_l [{shapeType=Capsule, length=20, radius=4.5, shapeName=upperarm_l_capsule, center=<10.4 0 0>, rotation=<rot 90 0 0>}]
 upperarm_r [{shapeType=Capsule, length=20, radius=4.5, shapeName=upperarm_r_capsule, center=<-10.4 0 0>, rotation=<rot 90 0 0>}]
}
constraints [
 {
  bone1Name=upperarm_l
  bone2Name=clavicle_l
  swing1LimitDegrees=45
  swing1Motion=Free
  swing2LimitDegrees=45
  swing2Motion=Free
  twistLimitDegrees=45
  twistMotion=Free
 }
 {
  bone1Name=upperarm_r
  bone2Name=clavicle_r
  swing1LimitDegrees=45
  swing1Motion=Free
  swing2LimitDegrees=45
  swing2Motion=Free
  twistLimitDegrees=45
  twistMotion=Free
 }
 {bone1Name=base_hair_01_l, bone2Name=head, swing1LimitDegrees=1, swing2LimitDegrees=1, twistLimitDegrees=1}
 {bone1Name=dyn_hair_01_l, bone2Name=base_hair_01_l, swing1LimitDegrees=30, swing2LimitDegrees=30, twistLimitDegrees=1}
 {bone1Name=dyn_hair_02_l, bone2Name=dyn_hair_01_l, swing1LimitDegrees=30, swing2LimitDegrees=30, twistLimitDegrees=1}
 {bone1Name=dyn_hair_03_l, bone2Name=dyn_hair_02_l, swing1LimitDegrees=35, swing2LimitDegrees=35, twistLimitDegrees=1}
 {bone1Name=base_hair_01_r, bone2Name=head, swing1LimitDegrees=1, swing2LimitDegrees=1, twistLimitDegrees=1}
 {bone1Name=dyn_hair_01_r, bone2Name=base_hair_01_r, swing1LimitDegrees=30, swing2LimitDegrees=30, twistLimitDegrees=1}
 {bone1Name=dyn_hair_02_r, bone2Name=dyn_hair_01_r, swing1LimitDegrees=30, swing2LimitDegrees=30, twistLimitDegrees=1}
 {bone1Name=dyn_hair_03_r, bone2Name=dyn_hair_02_r, swing1LimitDegrees=35, swing2LimitDegrees=35, twistLimitDegrees=1}
]
