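# Post-copy anim blueprint for the Cyber Fu female face accessory part, adding secondary motion to hair and earring.
# Copies the parent pose, layers Body_Mesh's pose from the upper arms, then rotates right hair while targeting.
# A head-space RigidBody with state-driven gravity feeds left-earring AnimDynamics and small bone offset and scale fixes.
# A separate hair AnimDynamics chain is unconnected to the output.
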AnimBlueprint /BRCosmetics/Characters/Player/Female/Medium/Bodies/F_MED_Cyber_Fu/Meshes/Parts/F_MED_Cyber_Fu_FaceAcc_AnimBP
target_skeleton=&/F_MED_Cyber_Fu_FaceAcc_Skeleton
graphs [
 AnimGraph {nodes [
   n1 = Anim Dynamics - Bone: dyn_hair_01_l(Alpha=1, ExternalForce="0,0,0") {
    angularDampingOverride=1
    angularSpringConstant=1500
    bAngularSpring=true
    bGravityOverrideInSimSpace=true
    bOverrideAngularDamping=true
    bUseSphericalLimits=true
    gravityScale=0.29
    simulationSpace=RootRelative
    alphaBoolBlend {blendOption=Linear}
    boundBone {boneName=dyn_hair_01_l}
    componentAppliedLinearAccClamp=<3000 3000 3000>
    componentLinearAccScale=<0.3 0.3 0.6>
    componentLinearVelScale=<0.3 0.3 0.6>
    physicsBodyDefinitions [{
      collisionType=CustomSphere
      sphereCollisionRadius=1.5
      boundBone {boneName=dyn_hair_01_l}
      boxExtents=<13 20 20>
      constraintSetup {angularLimitsMax=<0 10 3>, angularLimitsMin=<0 -35 -40>, angularTarget=<0 0 0>}
      localJointOffset=<10 0 0>
     }]
    sphericalLimits [
     {limitRadius=8, drivingBone {boneName=head}, sphereLocalOffset=<0 0 0>}
     {limitRadius=8, drivingBone {boneName=head}, sphereLocalOffset=<5 -4 0>}
    ]
   }
   n2 = Anim Dynamics - Bone: dyn_hair_02_l(ComponentPose=@n1, Alpha=1, ExternalForce="0,0,0") {
    angularDampingOverride=1
    angularSpringConstant=1500
    bAngularSpring=true
    bGravityOverrideInSimSpace=true
    bOverrideAngularDamping=true
    bUseSphericalLimits=true
    gravityScale=0.29
    simulationSpace=RootRelative
    alphaBoolBlend {blendOption=Linear}
    boundBone {boneName=dyn_hair_02_l}
    componentAppliedLinearAccClamp=<3000 3000 3000>
    componentLinearAccScale=<0.3 0.3 0.6>
    componentLinearVelScale=<0.3 0.3 0.6>
    physicsBodyDefinitions [{
      collisionType=CustomSphere
      sphereCollisionRadius=1.5
      boundBone {boneName=dyn_hair_02_l}
      boxExtents=<13 20 20>
      constraintSetup {angularLimitsMax=<0 10 6>, angularLimitsMin=<0 -25 -40>, angularTarget=<0 0 0>}
      localJointOffset=<10 0 0>
     }]
    sphericalLimits [
     {limitRadius=8, drivingBone {boneName=head}, sphereLocalOffset=<0 0 0>}
     {limitRadius=8, drivingBone {boneName=head}, sphereLocalOffset=<5 -4 0>}
    ]
   }
   n3 = Anim Dynamics - Bone: dyn_hair_03_l(ComponentPose=@n2, Alpha=1, ExternalForce="0,0,0") {
    angularDampingOverride=1
    angularSpringConstant=1500
    bAngularSpring=true
    bGravityOverrideInSimSpace=true
    bOverrideAngularDamping=true
    bUseSphericalLimits=true
    gravityScale=0.29
    simulationSpace=RootRelative
    alphaBoolBlend {blendOption=Linear}
    boundBone {boneName=dyn_hair_03_l}
    componentAppliedLinearAccClamp=<3000 3000 3000>
    componentLinearAccScale=<0.3 0.3 0.6>
    componentLinearVelScale=<0.3 0.3 0.6>
    physicsBodyDefinitions [{
      collisionType=CustomSphere
      sphereCollisionRadius=1.5
      boundBone {boneName=dyn_hair_03_l}
      boxExtents=<13 20 20>
      constraintSetup {angularLimitsMax=<0 15 5>, angularLimitsMin=<0 -25 -40>, angularTarget=<0 0 0>}
      localJointOffset=<10 0 0>
     }]
    sphericalLimits [
     {limitRadius=8, drivingBone {boneName=head}, sphereLocalOffset=<0 0 0>}
     {limitRadius=8, drivingBone {boneName=head}, sphereLocalOffset=<5 -4 0>}
    ]
   }
   n4 = Anim Dynamics - Bone: dyn_hair_01_r(ComponentPose=@n3, Alpha=1, ExternalForce="0,0,0") {
    angularDampingOverride=1
    angularSpringConstant=1500
    bAngularSpring=true
    bGravityOverrideInSimSpace=true
    bOverrideAngularDamping=true
    bUseSphericalLimits=true
    gravityScale=0.29
    simulationSpace=RootRelative
    alphaBoolBlend {blendOption=Linear}
    boundBone {boneName=dyn_hair_01_r}
    componentAppliedLinearAccClamp=<3000 3000 3000>
    componentLinearAccScale=<0.3 0.3 0.6>
    componentLinearVelScale=<0.3 0.3 0.6>
    physicsBodyDefinitions [{
      collisionType=CustomSphere
      sphereCollisionRadius=1.5
      boundBone {boneName=dyn_hair_01_r}
      boxExtents=<13 20 20>
      constraintSetup {angularLimitsMax=<0 10 3>, angularLimitsMin=<0 -35 -40>, angularTarget=<0 0 0>}
      localJointOffset=<-10 0 0>
     }]
    sphericalLimits [
     {limitRadius=8, drivingBone {boneName=head}, sphereLocalOffset=<0 0 0>}
     {limitRadius=8, drivingBone {boneName=head}, sphereLocalOffset=<5 -4 0>}
    ]
   }
   n5 = Anim Dynamics - Bone: dyn_hair_02_r(ComponentPose=@n4, Alpha=1, ExternalForce="0,0,0") {
    angularDampingOverride=1
    angularSpringConstant=1500
    bAngularSpring=true
    bGravityOverrideInSimSpace=true
    bOverrideAngularDamping=true
    bUseSphericalLimits=true
    gravityScale=0.29
    simulationSpace=RootRelative
    alphaBoolBlend {blendOption=Linear}
    boundBone {boneName=dyn_hair_02_r}
    componentAppliedLinearAccClamp=<3000 3000 3000>
    componentLinearAccScale=<0.3 0.3 0.6>
    componentLinearVelScale=<0.3 0.3 0.6>
    physicsBodyDefinitions [{
      collisionType=CustomSphere
      sphereCollisionRadius=1.5
      boundBone {boneName=dyn_hair_02_r}
      boxExtents=<13 20 20>
      constraintSetup {angularLimitsMax=<0 10 6>, angularLimitsMin=<0 -25 -40>, angularTarget=<0 0 0>}
      localJointOffset=<-10 0 0>
     }]
    sphericalLimits [
     {limitRadius=8, drivingBone {boneName=head}, sphereLocalOffset=<0 0 0>}
     {limitRadius=8, drivingBone {boneName=head}, sphereLocalOffset=<5 -4 0>}
    ]
   }
   n6 = Anim Dynamics - Bone: dyn_hair_03_r(ComponentPose=@n5, Alpha=1, ExternalForce="0,0,0") {
    angularDampingOverride=1
    angularSpringConstant=1500
    bAngularSpring=true
    bGravityOverrideInSimSpace=true
    bOverrideAngularDamping=true
    bUseSphericalLimits=true
    gravityScale=0.29
    simulationSpace=RootRelative
    alphaBoolBlend {blendOption=Linear}
    boundBone {boneName=dyn_hair_03_r}
    componentAppliedLinearAccClamp=<3000 3000 3000>
    componentLinearAccScale=<0.3 0.3 0.6>
    componentLinearVelScale=<0.3 0.3 0.6>
    physicsBodyDefinitions [{
      collisionType=CustomSphere
      sphereCollisionRadius=1.5
      boundBone {boneName=dyn_hair_03_r}
      boxExtents=<13 20 20>
      constraintSetup {angularLimitsMax=<0 15 5>, angularLimitsMin=<0 -25 -40>, angularTarget=<0 0 0>}
      localJointOffset=<-10 0 0>
     }]
    sphericalLimits [
     {limitRadius=8, drivingBone {boneName=head}, sphereLocalOffset=<0 0 0>}
     {limitRadius=8, drivingBone {boneName=head}, sphereLocalOffset=<5 -4 0>}
    ]
   }
   Anim Dynamics - Bone: dyn_hair_back(ComponentPose=@n6, Alpha=1, ExternalForce="0,0,0") {
    angularDampingOverride=1
    angularSpringConstant=1500
    bAngularSpring=true
    bGravityOverrideInSimSpace=true
    bOverrideAngularDamping=true
    bUseSphericalLimits=true
    gravityScale=0.29
    simulationSpace=RootRelative
    alphaBoolBlend {blendOption=Linear}
    boundBone {boneName=dyn_hair_back}
    componentAppliedLinearAccClamp=<3000 3000 3000>
    componentLinearAccScale=<0.3 0.3 0.6>
    componentLinearVelScale=<0.3 0.3 0.6>
    physicsBodyDefinitions [{
      sphereCollisionRadius=1.5
      boundBone {boneName=dyn_hair_back}
      boxExtents=<13 20 20>
      constraintSetup {angularLimitsMax=<10 15 5>, angularLimitsMin=<-25 -25 -40>, angularTarget=<0 0 0>}
      localJointOffset=<0 -0.8 0>
     }]
   }
   n7 = Copy Pose From Mesh(SourceMeshComponent=Get Body_Mesh())
   n8 = Copy Pose From Mesh() {bCopyCurves=true, bUseAttachedParent=true}
   n9 = "Transform (Modify) Bone - Bone: base_hair_01_r"(
    ComponentPose=Local To Component(LocalPose=Layered blend per bone(BasePose=@n8, BlendPoses_0=@n7) {layerSetup [{branchFilters [{boneName=upperarm_l}, {boneName=upperarm_r}]}]})
    Alpha=Get IsTargeting() -> (IsTargeting:double)
    Translation="0,0,0"
    Rotation="0,30,0"
    Scale="1,1,1"
   ) {
    rotationMode=BMM_Additive
    rotationSpace=BCS_BoneSpace
    alphaBoolBlend {blendOption=Linear}
    alphaScaleBiasClamp {bInterpResult=true}
    boneToModify {boneName=base_hair_01_r}
    rotation {yaw=30}
   }
   n10 = RigidBody!(ComponentPose=@n9, Alpha, OverrideWorldGravity:Vector=Get GravityOverride(), ExternalForce) -> (Pose) {
    bClampLinearTranslationLimitToRefPose=true
    bOverrideWorldGravity=true
    overridePhysicsAsset=&/F_MED_Cyber_Fu_FaceAcc_Physics
    simulationSpace=BaseBoneSpace
    alphaBoolBlend {blendOption=Linear}
    baseBoneRef {boneName=head}
    componentAppliedLinearAccClamp=<3000 3000 3000>
    componentLinearAccScale=<0.35 0.2 0.2>
    componentLinearVelScale=<0.35 0.2 0.2>
    overrideWorldGravity {z=-980}
   }
   n11 = "Transform (Modify) Bone - Bone: dyn_earring_01_l"(ComponentPose=@n10, Alpha=1, Translation="0.55,0,0.15", Rotation="0,0,0", Scale="1,1,1") {
    translationMode=BMM_Additive
    translationSpace=BCS_BoneSpace
    alphaBoolBlend {blendOption=Linear}
    boneToModify {boneName=dyn_earring_01_l}
    translation {x=0.55, z=0.15}
   }
   n12 = Anim Dynamics - Bone: dyn_earring_01_l(ComponentPose=@n11, Alpha=1, ExternalForce="0,0,0") {
    angularDampingOverride=1
    angularSpringConstant=1500
    bAngularSpring=true
    bGravityOverrideInSimSpace=true
    bOverrideAngularDamping=true
    bUseSphericalLimits=true
    gravityScale=0.35
    simulationSpace=RootRelative
    alphaBoolBlend {blendOption=Linear}
    boundBone {boneName=dyn_earring_01_l}
    componentAppliedLinearAccClamp=<3000 3000 3000>
    componentLinearAccScale=<0.3 0.3 0.6>
    componentLinearVelScale=<0.3 0.3 0.6>
    physicsBodyDefinitions [{
      sphereCollisionRadius=1.5
      boundBone {boneName=dyn_earring_01_l}
      boxExtents=<5 5 5>
      constraintSetup {angularLimitsMax=<15 10 3>, angularLimitsMin=<-15 -10 -3>, angularTarget=<0 0 0>}
      localJointOffset=<0 0 -1.2>
     }]
   }
   n13 = "Transform (Modify) Bone - Bone: jaw"(ComponentPose=@n12, Alpha=1, Translation="0,0,0", Rotation="0,0,0", Scale="1.02,1.02,1.02") {scaleMode=BMM_Additive, scaleSpace=BCS_BoneSpace, alphaBoolBlend {blendOption=Linear}, boneToModify {boneName=jaw}, scale=<1.02 1.02 1.02>}
   n14 = "Transform (Modify) Bone - Bone: hair"(ComponentPose=@n13, Alpha=1, Translation="0.015,0,0", Rotation="0,0,0", Scale="1,1,1") {
    translationMode=BMM_Additive
    translationSpace=BCS_BoneSpace
    alphaBoolBlend {blendOption=Linear}
    boneToModify {boneName=hair}
    translation {x=0.015}
   }
   Output Pose(Result=Component To Local(ComponentPose=@n14)) {layerGroup=None}
  ]}
 EventGraph {nodes [
   Event BlueprintUpdateAnimation()
   > n1 = Sequence()
   > Set IsTargeting(IsTargeting:double="To Float (Boolean)"(InBool=Get bIsTargeting())) -> (Output_Get:double)
   n2 = Gravity Override State Function DEPRECATED(
    execute=@n1.then_1
    CharacterPartAnimInstance=Self-Reference()
    GameTimeInSeconds=GetGameTimeInSeconds()
    GravityOverrideParams=Get Gravity Override Params()
   )
   > Set GravityOverride(GravityOverride=@n2)
   Event OnCharacterCustomizationLoaded()
   > n3 = Cast To PlayerPawn_Athena(Object=TryGetPawnOwner())
   > Set Body_Mesh(Body_Mesh=GetSkeletalMeshForPartType(self=@n3))
   TryGetPawnOwner()
  ]}
]
variables [
 Body_Mesh {type=SkeletalMeshComponent}
 Gravity Override Params {type=GravityOverrideParamsStruct, default {skydive_FloatingDown {gravityOverride=<100 100 1000>}}}
 GravityOverride {type=Vector}
 IsTargeting {type=number}
]
